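# Master skeletal-mesh character material for the male medium Stands body, with optional music-reactive glow.
# Shared PreFX and CharacterSkinFX functions build attributes, which pass through PostFX and cinematic PostFX functions.
# The "Stands Audio Reactivity" switch maps music spectrum values along pre-skinned height, masks them, then blends in Fresnel-shaded emissive color and darkens base color.
# Amplitude, hue-shift and HSV-blend branches are left unconnected.
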Material /BRCosmetics/Characters/Player/Male/Medium/Bodies/M_MED_Stands/Materials/M_Stands_MASTER
graph {
 properties {
  bAutomaticallySetUsageInEditor=false
  bUseMaterialAttributes=true
  bUsedWithMeshDeformer=true
  bUsedWithMorphTargets=true
  bUsedWithSkeletalMesh=true
  pixelDepthOffsetMode=PDOM_Legacy
  refractionMethod=RM_IndexOfRefraction
  shadingModel=MSM_FromMaterialExpression
  subsurfaceProfile=&SS_HeroSkin_02
  usedShadingModels=Default Lit | Subsurface | Clear Coat | Subsurface Profile | Cloth | Eye
 }
 nodes [
  n1 = TextureSampleParameter2D() {
   parameterName=AudioReactivityMask
   samplerType=SAMPLERTYPE_Masks
   texture=/Game/Characters/Player/Male/Medium/Bodies/M_MED_Stands/Textures/T_M_MED_Stands_Body_Mask
  }
  n2 = CollectionParameter() {collection=/Game/Athena/Events/AudioAnalysis/MPC_Audio_Reactive_Parameters, parameterName=Spectrum_Music}
  n3 = MaterialFunctionCall(Color=@n2) {materialFunction=&SplitComponents}
  n4 = ComponentMask(None=@n2) {a=true}
  n5 = PreSkinnedPosition()
  n6 = PreSkinnedLocalBounds()
  n7 = VertexInterpolator(VS=Divide(A=ComponentMask(None=@n5) {b=true}, B=ComponentMask(None=@n6.Max) {b=true}))
  n8 = Multiply(
   A=CollectionParameter() {collection=&MPC_Audio_Reactive_Parameters, parameterName=Amplitude_Music}
   B=Power(
    Base=MaterialFunctionCall!(1=@n3.R, 2=@n3.R, 3=@n3.R, 4=@n4, A=@n7.PS) {materialFunction=/Engine/Functions/Engine_MaterialFunctions02/Lerp_Multiple_Float4}.Lerp 4 Inputs
    Exp=ScalarParameter() {defaultValue=0.5, parameterName=AudioReactivityExponent}
   ) {constExponent=1}
  )
  n9 = ScalarParameter() {defaultValue=1, parameterName=FresnelExponent}
  n10 = Add(A=OneMinus(None=Fresnel(ExponentIn=@n9) {baseReflectFraction=0.5, exponent=0.4}))
  n11 = Saturate(None=Power(Base=Multiply(A=@n8, B=@n10), Exp=ScalarParameter() {defaultValue=0.7, parameterName=AudioExponent}))
  n12 = Multiply(A=ScalarParameter() {parameterName=Shift Max}, B=Power(Base=@n11) {constExponent=0.5})
  n13 = VectorParameter() {parameterName=Audio Secondary Color, defaultValue=<rgba 0.8 0.8 0.8 1>}
  n14 = LinearInterpolate(A=VectorParameter() {parameterName=Audio Primary Color, defaultValue=<rgba 0.9271 0.0922 0.5072 1>}.RGB, B=@n13.RGB) {constAlpha=0}
  n15 = Multiply(
   A=LinearInterpolate(
    A=MaterialFunctionCall(Hue Shift Percentage=@n12, Texture=@n14) {materialFunction=&HueShift}.Result
    B=@n13.RGB
    Alpha=@n11
   )
   B=LinearInterpolate(Alpha=Power(Base=@n8) {constExponent=0.75}) {constA=0.5}
  )
  LinearInterpolate(B=@n15, Alpha=@n1.R)
  LinearInterpolate(B=@n15, Alpha=@n1.R) {constB=0.01}
  n16 = Reroute() {desc=Audio input}
  n17 = MaterialFunctionCall!(RGB) {materialFunction=/Engine/Functions/Engine_MaterialFunctions02/Math/RGBtoHSV}
  n18 = MaterialFunctionCall() {materialFunction=&RGBtoHSV}
  n19 = Subtract(A=ComponentMask(None=@n17."RGB to HSV (0-1)") {r=true}, B=ComponentMask(None=@n18."RGB to HSV (0-1)") {r=true})
  n20 = Abs(None=@n19)
  n21 = Constant() {r=0.5}
  n22 = Sign(None=@n19)
  n23 = Constant()
  n24 = Add(
   A=MaterialFunctionCall(X=If(A=@n20, B=@n21, "A > B"=@n22, "A == B"=@n23, "A < B"=@n23)) {materialFunction=&MakeFloat3}.Result
   B=@n18."RGB to HSV (0-1)"
  )
  LinearInterpolate(
   A=LinearInterpolate(
    A=VectorParameter() {parameterName=Audio Primary Color, defaultValue=<rgba 0.9271 0.0922 0.5072 1>}.RGB
    B=VectorParameter() {parameterName=Audio Secondary Color, defaultValue=<rgba 0.8 0.8 0.8 1>}.RGB
    Alpha=@n16
   )
   B=MaterialFunctionCall!(HSV_In=LinearInterpolate(A=@n17."RGB to HSV (0-1)", B=@n24, Alpha=@n16) {constAlpha=1}) {materialFunction=/Game/Athena/UI/Materials/MF_HSVtoRGB}.RGB_Out
   Alpha=Min(A=ComponentMask(None=@n17."RGB to HSV (0-1)") {g=true}, B=ComponentMask(None=@n18."RGB to HSV (0-1)") {g=true})
  )
  n25 = MaterialFunctionCall() {materialFunction=&BaseFN_CharacterFunction_PreFX}
  n26 = GetMaterialAttributes(None=MaterialFunctionCall(
    PreskinnedPosition=@n25.Pre skinned Local Position
    Gmap=@n25.Pre skinned Local Position
    MaterialAttributes=@n25.Pre skinned Local Position
    AO=@n25.Pre skinned Local Position
   ) {materialFunction=&CharacterSkinFX}.Result) {attributeGetTypes [#1, #2]}
  n27 = Fresnel() {baseReflectFraction=0}
  n28 = TextureSampleParameter2D() {
   parameterName=AudioReactivityMask
   samplerType=SAMPLERTYPE_Masks
   texture=&T_M_MED_Stands_Body_Mask
  }
  n29 = CollectionParameter() {collection=&MPC_Audio_Reactive_Parameters, parameterName=Spectrum_Music}
  n30 = MaterialFunctionCall(Color=@n29) {materialFunction=&SplitComponents}
  n31 = ComponentMask(None=PreSkinnedPosition()) {b=true}
  n32 = ComponentMask(None=PreSkinnedLocalBounds().Max) {b=true}
  n33 = MaterialFunctionCall(1=@n30.R, 2=@n30.R, 3=@n30.R, 4=ComponentMask(None=@n29) {a=true}, A=VertexInterpolator(VS=Divide(A=@n31, B=@n32)).PS) {materialFunction=&Lerp_Multiple_Float4}
  n34 = Multiply(A=InverseLinearInterpolate(Value=@n33.Lerp 4 Inputs) {constA=0.45, constB=0.9}) {constB=1.2}
  n35 = ScalarParameter() {parameterName=Audio Offset}
  n36 = Multiply(A=@n28.R, B=Saturate(None=Add(A=@n34, B=@n35)))
  n37 = LinearInterpolate(
   A=@n26.Emissive Color
   B=Multiply(
    A=VectorParameter() {parameterName=Audio Primary Color, defaultValue=<rgba 0.9271 0.0922 0.5072 1>}.RGB
    B=LinearInterpolate(Alpha=@n27) {constA=2.1, constB=0.8}
   ) {constB=1.5}
   Alpha=@n36
  )
  n38 = LinearInterpolate(A=@n26.Base Color, Alpha=@n36) {constB=0.01}
  n39 = MaterialFunctionCall(
   AO=@n25.AO
   Distance Blend=@n25.AO
   Fresnel=@n25.AO
   Gmap=@n25.AO
   Pre Skinned Local Position=@n25.AO
   Material=StaticSwitchParameter(
    True=SetMaterialAttributes!(MaterialAttributes=@n26.MaterialAttributes, Emissive Color=@n37, Base Color=@n38) {attributeSetTypes [#1, #2]}
    False=@n26.MaterialAttributes
   ) {parameterName=Stands Audio Reactivity}
  ) {materialFunction=&BaseFN_CharacterFunction_PostFX}
  MaterialFunctionCall(In=@n39.Result) {materialFunction=&CINE_MF_CharPostFX}
  n40 = ScalarParameter() {defaultValue=0.5, parameterName=AudioReactivityExponent}
  Power(
   Base=Multiply(
    A=CollectionParameter() {collection=&MPC_Audio_Reactive_Parameters, parameterName=Amplitude_Music}
    B=Power(Base=@n33.Lerp 4 Inputs, Exp=@n40) {constExponent=1}
   )
   Exp=ScalarParameter() {defaultValue=0.7, parameterName=AudioExponent}
  )
 ]
}
