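# Physics asset for the M_MED_Stands male medium player body cosmetic, covering secondary motion of the hoodie drawstrings and front pocket.
# Each drawstring is a three-link capsule chain with limited swing and locked twist, hung from a base body that is angularly locked to spine_05.
# The pocket swings narrowly on its master body, which links to spine_01 with free rotation, as does spine_05, the largest torso capsule.
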PhysicsAsset /BRCosmetics/Characters/Player/Male/Medium/Bodies/M_MED_Stands/Meshes/M_MED_Stands_Physics
bodies {
 dyn_L_hoodieString_1 [{shapeType=Capsule, length=3, radius=1.3, center=<0.8 0 0>, rotation=<rot 90 0 0>}]
 dyn_L_hoodieString_2 [{shapeType=Capsule, length=3, radius=0.9, center=<0.8 0 0>, rotation=<rot 90 0 0>}]
 dyn_L_hoodieString_3 [{shapeType=Capsule, length=3, radius=0.8, center=<0.8 0 0>, rotation=<rot 90 0 0>}]
 dyn_L_hoodieString_base [{shapeType=Capsule, length=0.505, radius=0.505, rotation=<rot 0 -90 0>}]
 dyn_R_hoodieString_1 [{shapeType=Capsule, length=3, radius=1.3, center=<-0.8 0 0>, rotation=<rot 90 0 0>}]
 dyn_R_hoodieString_2 [{shapeType=Capsule, length=3, radius=0.9, center=<-0.8 0 0>, rotation=<rot 90 0 0>}]
 dyn_R_hoodieString_3 [{shapeType=Capsule, length=3, radius=0.8, center=<-0.8 0 0>, rotation=<rot 90 0 0>}]
 dyn_R_hoodieString_base [{shapeType=Capsule, length=0.505, radius=0.505, rotation=<rot 0 -90 0>}]
 dyn_frontPocket_jnt [{shapeType=Capsule, length=12.4169, radius=4, center=<-0.8166 0.0001 0>, rotation=<rot 0 0 90>}]
 dyn_frontPocket_jnt_master [{shapeType=Capsule, length=0.505, radius=0.505, rotation=<rot 0 -90 0>}]
 spine_01 [{shapeType=Capsule, length=1, radius=1, center=<6.6848 -0.4973 -0.6202>, rotation=<rot 1.0211 -90 -0.786>}]
 spine_05 [{shapeType=Capsule, length=14.3574, radius=10, center=<-5.0997 2.0186 0.5068>}]
}
constraints [
 {
  bone1Name=dyn_frontPocket_jnt
  bone2Name=dyn_frontPocket_jnt_master
  swing1LimitDegrees=5
  swing2LimitDegrees=5
  twistLimitDegrees=45
  twistMotion=Locked
 }
 {
  bone1Name=dyn_R_hoodieString_1
  bone2Name=dyn_R_hoodieString_base
  swing1LimitDegrees=45
  swing2LimitDegrees=45
  twistLimitDegrees=45
  twistMotion=Locked
 }
 {
  bone1Name=dyn_R_hoodieString_2
  bone2Name=dyn_R_hoodieString_1
  swing1LimitDegrees=45
  swing2LimitDegrees=45
  twistLimitDegrees=45
  twistMotion=Locked
 }
 {
  bone1Name=dyn_R_hoodieString_3
  bone2Name=dyn_R_hoodieString_2
  swing1LimitDegrees=60
  swing2LimitDegrees=45
  twistLimitDegrees=45
  twistMotion=Locked
 }
 {
  bone1Name=dyn_L_hoodieString_1
  bone2Name=dyn_L_hoodieString_base
  swing1LimitDegrees=45
  swing2LimitDegrees=45
  twistLimitDegrees=45
  twistMotion=Locked
 }
 {
  bone1Name=dyn_L_hoodieString_2
  bone2Name=dyn_L_hoodieString_1
  swing1LimitDegrees=45
  swing2LimitDegrees=45
  twistLimitDegrees=45
  twistMotion=Locked
 }
 {
  bone1Name=dyn_L_hoodieString_3
  bone2Name=dyn_L_hoodieString_2
  swing1LimitDegrees=60
  swing2LimitDegrees=45
  twistLimitDegrees=45
  twistMotion=Locked
 }
 {
  bone1Name=spine_05
  bone2Name=spine_01
  swing1LimitDegrees=45
  swing1Motion=Free
  swing2LimitDegrees=45
  swing2Motion=Free
  twistLimitDegrees=45
  twistMotion=Free
 }
 {
  bone1Name=dyn_L_hoodieString_base
  bone2Name=spine_05
  swing1LimitDegrees=45
  swing1Motion=Locked
  swing2LimitDegrees=45
  swing2Motion=Locked
  twistLimitDegrees=45
  twistMotion=Locked
 }
 {
  bone1Name=dyn_R_hoodieString_base
  bone2Name=spine_05
  swing1LimitDegrees=45
  swing1Motion=Locked
  swing2LimitDegrees=45
  swing2Motion=Locked
  twistLimitDegrees=45
  twistMotion=Locked
 }
 {
  bone1Name=dyn_frontPocket_jnt_master
  bone2Name=spine_01
  swing1LimitDegrees=45
  swing1Motion=Free
  swing2LimitDegrees=45
  swing2Motion=Free
  twistLimitDegrees=45
  twistMotion=Free
 }
]
