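# Animation Blueprint for the M_MED_Stands body cosmetic, adding secondary motion to a pose copied from the attached parent mesh, mainly for the hood.
# RigidBody and AnimDynamics simulation feed two additive bone offsets on the hood joints.
# The event graph sets the offset vector when crouching, zero otherwise, and derives its alpha from head yaw relative to spine_05.
# A neck-relative hoodAlpha drives only a disconnected transform node.
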AnimBlueprint /BRCosmetics/Characters/Player/Male/Medium/Bodies/M_MED_Stands/Meshes/M_MED_Stands_AnimBP
target_skeleton=&/M_MED_Stands_Skeleton
graphs [
 AnimGraph {nodes [
   "Transform (Modify) Bone - Bone: dyn_hoodDip_jnt"(Alpha=Get hoodAlpha() -> (hoodAlpha:double), Translation="0,-1,-1.5", Rotation="0,0,0", Scale="1,1,1") {translationMode=BMM_Additive, alphaBoolBlend {blendOption=Linear}, boneToModify {boneName=dyn_hoodDip_jnt}, translation {y=-1, z=-1.5}}
   n1 = RigidBody(ComponentPose=Local To Component(LocalPose=Copy Pose From Mesh() {bUseAttachedParent=true})) {
    bClampLinearTranslationLimitToRefPose=true
    bOverrideWorldGravity=true
    overridePhysicsAsset=&/M_MED_Stands_Physics
    alphaBoolBlend {blendOption=Linear}
    baseBoneRef {boneName=pelvis}
    componentAppliedLinearAccClamp=<5000 5000 5000>
    componentLinearAccScale=<0.3 0.3 0.3>
    componentLinearVelScale=<0.2 0.2 0.2>
    overrideWorldGravity {z=-800}
   }
   n2 = Anim Dynamics - Bone: dyn_hood_jnt(ComponentPose=@n1, Alpha=1, ExternalForce="0,0,0") {
    angularDampingOverride=0.999
    bGravityOverrideInSimSpace=true
    bOverrideAngularDamping=true
    bOverrideLinearDamping=true
    linearDampingOverride=0.999
    simulationSpace=RootRelative
    alphaBoolBlend {blendOption=Linear}
    boundBone {boneName=dyn_hood_jnt}
    componentLinearAccScale=<1 1 1>
    componentLinearVelScale {x=0.2}
    physicsBodyDefinitions [{
      sphereCollisionRadius=0
      boundBone {boneName=dyn_hood_jnt}
      boxExtents=<20 7 7>
      constraintSetup {angularLimitsMax=<1 1 1>, angularLimitsMin=<-1 -1 -1>, angularTarget=<0 0 0>, linearAxesMax=<0.5 0.7 0>, linearAxesMin=<-0.5 -0.7 0>}
      localJointOffset=<0 0 0.1>
     }]
   }
   n3 = Get crouchVector() -> (crouchVector:Vector)
   n4 = Get hoodAlphaCrouch() -> (hoodAlphaCrouch:double)
   n5 = "Transform (Modify) Bone - Bone: dyn_hoodDip_jnt"(ComponentPose=@n2, Alpha=@n4, Translation=@n3, Rotation="0,0,0", Scale="1,1,1") {
    translationMode=BMM_Additive
    alphaBoolBlend {blendOption=Linear}
    alphaScaleBiasClamp {bInterpResult=true}
    boneToModify {boneName=dyn_hoodDip_jnt}
    translation {y=-2, z=-3.5}
   }
   n6 = "Transform (Modify) Bone - Bone: dyn_hood_jnt"(ComponentPose=@n5, Alpha=@n4, Translation=@n3, Rotation="0,0,0", Scale="1,1,1") {
    translationMode=BMM_Additive
    alphaBoolBlend {blendOption=Linear}
    alphaScaleBiasClamp {bInterpResult=true}
    boneToModify {boneName=dyn_hood_jnt}
    translation {y=-1, z=-1}
   }
   Output Pose(Result=Component To Local(ComponentPose=@n6)) {layerGroup=None}
   Get crouchVector()
  ]}
 EventGraph {nodes [
   Event BlueprintUpdateAnimation()
   > n1 = Sequence!(execute) -> (then_0, then_1, then_2:Exec)
   n2 = GetOwnerSkeletalMeshComponent()
   n3 = BreakTransform(InTransform=GetDeltaTransformFromRefPose(self=@n2, BoneName=head, BaseName=neck_01))
   > Set hoodAlpha(hoodAlpha:double=MapRangeClamped(Value=BreakRotator(InRot=@n3.Rotation).Yaw, InRangeA=-15, InRangeB=-27)) -> (Output_Get:double)
   n4 = BreakTransform(InTransform=GetDeltaTransformFromRefPose(self=@n2, BoneName=head, BaseName=spine_05))
   Set hoodAlphaCrouch(execute=@n1.then_1, hoodAlphaCrouch:double=MapRangeClamped(Value=BreakRotator(InRot=@n4.Rotation).Yaw, InRangeA=-30, InRangeB=-60)) -> (Output_Get:double)
   n5 = Branch(execute=@n1.then_2, Condition=Get bIsCrouching())
   > Set crouchVector(crouchVector:Vector="0,-0.7,-1.5") -> (Output_Get:Vector)
   Set crouchVector(execute=@n5.else, crouchVector="0, 0, 0")
  ]}
]
variables [
 crouchVector {type=Vector}
 hoodAlpha {type=number}
 hoodAlphaCrouch {type=number}
]
